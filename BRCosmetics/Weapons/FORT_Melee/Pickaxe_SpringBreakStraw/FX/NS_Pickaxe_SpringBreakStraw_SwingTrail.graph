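# Niagara swing trail for the SpringBreakStraw pickaxe, built from two ribbon emitters, BladeTrail and Light_Trail.
# Both spawn short-lived anim-trail particles from skeletal mesh bones, with spawn alpha pulsed by a curve over system age.
# BladeTrail is a white ribbon; Light_Trail is wider, narrows over its life and takes its tint from a time-of-day color curve.
# Both fade out over particle life, with opacity scaled by time of day.
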NiagaraSystem /BRCosmetics/Weapons/FORT_Melee/Pickaxe_SpringBreakStraw/FX/NS_Pickaxe_SpringBreakStraw_SwingTrail
dependencies {
 usedDataInterfaces [&NiagaraDataInterfaceCurve, &NiagaraDataInterfaceColorCurve, &NiagaraDataInterfaceSkeletalMesh]
 usedDynamicInputs []
 usedModules [
  &SystemState
  &self:SystemSpawnScript.SystemScriptSource.SystemScriptGraph.NiagaraNodeAssignment_1.SetVariables_#1
  &self:SystemSpawnScript.SystemScriptSource.SystemScriptGraph.NiagaraNodeAssignment_0.SetVariables_#2
 ]
}
emitters {
 BladeTrail {
  emitterUpdateScript [
   &EmitterState()
   &FN_SpawnAnimTrail()
  ]
  particleSpawnScript [
   &FN_InitializeAnimTrailBoneIndecies()
   &FN_InitializeAnimTrail(
    AnimTrailWidth=0.9
    Color=&MakeLinearColorFromVectorAndFloat(
     "Vector (RGB)"=<1 1 1>
     "Float (Alpha)"=&FloatFromCurve(
      FloatCurve=DataInterface() {propertyValues {Curve {keys [
          {arriveTangent=40.3796, interpMode=RCIM_Cubic, leaveTangent=40.3796, tangentMode=RCTM_User, time=0, value=0}
          <at 0.075 1 RCIM_Cubic>
          {interpMode=RCIM_Cubic, leaveTangent=1.3333, time=0.15, value=0}
         ]}}}
      CurveIndex=$System.Age
     )
    )
    Lifetime=0.15
    SkeletalMesh=$System.Skeletal Mesh
   )
  ]
  particleUpdateScript [
   &SolveForcesAndVelocity(Clamp Velocity=false, Limit Acceleration=false)
   &FN_UpdateAnimTrail(Locked Particle RibbonWidth=1, SkeletalMesh=$System.Skeletal Mesh) {module=FN_UpdateAnimTrail001}
   &ScaleColor(
    Scale Alpha=&FloatFromCurve(
     FloatCurve=DataInterface() {propertyValues {Curve {keys [
         {arriveTangent=-2.003, interpMode=RCIM_Cubic, leaveTangent=-2.003, tangentMode=RCTM_User, time=0.5, value=0.95}
         {interpMode=RCIM_Cubic, tangentMode=RCTM_User, time=1, value=0}
        ]}}}
     Scale Curve=&Multiply_Float(A=0.35, B=&FNTimeOfDayScalar(Night Scale=0.2, Morning Scale=0.8, Evening Scale=0.75))
    )
    Scale RGB=&VectorFromFloat(Value=1.5)
    ScaleA=true
    ScaleRGB=true
   )
  ]
  rendererClasses [&NiagaraRibbonRendererProperties]
  renderers [{rendererClass=&NiagaraRibbonRendererProperties}]
 }
 Light_Trail {
  emitterUpdateScript [
   &EmitterState()
   &FN_SpawnAnimTrail()
  ]
  particleSpawnScript [
   &FN_InitializeAnimTrailBoneIndecies()
   &FN_InitializeAnimTrail(
    AnimTrailWidth=1.1
    Color=&MultiplyLinearColor(
     Linear Color=<rgba 0.5521 0.5212 0.5118 2>
     Scale Factor=&MakeFloat4FromVectorAndFloat(
      "Vector (RGB)"=&VectorFromFloat(Value=1.7)
      "Float (Alpha)"=&Multiply_Float(
       A=2.5
       B=&FloatFromCurve(
        FloatCurve=DataInterface() {propertyValues {Curve {keys [
            {arriveTangent=51.813, interpMode=RCIM_Cubic, leaveTangent=51.813, tangentMode=RCTM_User, time=0, value=0}
            <at 0.07 1 RCIM_Cubic>
            <at 0.125 0 RCIM_Cubic>
           ]}}}
        CurveIndex=$System.Age
       )
      )
     )
    )
    Lifetime=0.125
    SkeletalMesh=$System.Skeletal Mesh
   )
  ]
  particleUpdateScript [
   &SolveForcesAndVelocity(Clamp Velocity=false, Limit Acceleration=false)
   &FN_UpdateAnimTrail(Locked Particle RibbonWidth=1, SkeletalMesh=$System.Skeletal Mesh)
   &ScaleRibbonWidth(Ribbon Width Scale=&FloatFromCurve(FloatCurve=DataInterface() {propertyValues {Curve {keys [{arriveTangent=-1.8706, interpMode=RCIM_Cubic, time=0, value=1}, {interpMode=RCIM_Cubic, tangentMode=RCTM_User, time=1, value=0.5}]}}}))
   &ScaleColor(
    Scale Alpha=&FloatFromCurve(
     FloatCurve=DataInterface() {propertyValues {Curve {keys [<at 0.1 1 RCIM_Cubic>,<at 1 0 RCIM_Cubic>]}}}
     Scale Curve=&FNTimeOfDayScalar(Night Scale=0.35, Morning Scale=0.8, Evening Scale=0.75)
    )
    Scale RGB=&Multiply_VectorByFloat(
     Vector=&MakeVectorFromLinearColorRGB(Color=&ColorFromCurve(
       ColorCurve=DataInterface() {propertyValues {
         AlphaCurve {keys [<at 0 1 RCIM_Cubic>]}
         BlueCurve {keys [<at 0.061 0.5885 RCIM_Cubic>,<at 0.25 0.3724>,<at 0.5 0.5463>,<at 0.75 0.3724>,<at 0.8984 0.5885 RCIM_Cubic>]}
         GreenCurve {keys [<at 0.061 0.2654 RCIM_Cubic>,<at 0.25 0.4752>,<at 0.5 0.6009>,<at 0.75 0.4752>,<at 0.8984 0.2654 RCIM_Cubic>]}
         RedCurve {keys [<at 0.061 0.19 RCIM_Cubic>,<at 0.25 0.6875>,<at 0.5 0.7135>,<at 0.75 0.6875>,<at 0.8984 0.1901 RCIM_Cubic>]}
        }}
       CurveIndex=Linked() {linkedVariable {name=NPC.FortniteNPC.TimeOfDay_0-1, type=NiagaraFloat}}
      ))
     Float=0.3
    )
    ScaleA=true
    ScaleRGB=true
   )
  ]
  rendererClasses [&NiagaraRibbonRendererProperties]
  renderers [{rendererClass=&NiagaraRibbonRendererProperties}]
 }
}
summary {
 systemName=NS_Pickaxe_SpringBreakStraw_SwingTrail
 emitters [
  {emitterName=BladeTrail, rendererClasses [&NiagaraRibbonRendererProperties]}
  {emitterName=Light_Trail, rendererClasses [&NiagaraRibbonRendererProperties]}
 ]
 userVariables [User.RibbonIDOffset {type=NiagaraInt32}]
}
