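# Physics asset for the offhand axe mesh of the SpringBreakStraw pickaxe, a Battle Royale melee cosmetic.
# It has a single body on the Melee_TwoHanded_Handle bone, made of one slightly offset and rotated capsule that serves as a simple collision shape.
# With no constraints or other bodies, the asset defines no jointed or articulated physics.
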PhysicsAsset /BRCosmetics/Weapons/FORT_Melee/Pickaxe_SpringBreakStraw/Meshes/SpringBreakStrawOffhand_Axe_PhysicsAsset
bodies {Melee_TwoHanded_Handle [{shapeType=Capsule, length=63.594, radius=7.1596, center=<-2.0889 1.2903 25.1191>, rotation=<rot 5.8574 -90 0.9147>}]}
constraints []
