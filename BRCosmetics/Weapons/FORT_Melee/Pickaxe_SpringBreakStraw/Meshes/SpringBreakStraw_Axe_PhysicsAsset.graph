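# Physics asset for the SpringBreakStraw pickaxe cosmetic's axe mesh, providing its simplified collision shape.
# A single capsule body is bound to the Melee_TwoHanded_Handle bone, elongated and slightly offset and rotated.
# There are no constraints, so the weapon is one body with no articulated parts.
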PhysicsAsset /BRCosmetics/Weapons/FORT_Melee/Pickaxe_SpringBreakStraw/Meshes/SpringBreakStraw_Axe_PhysicsAsset
bodies {Melee_TwoHanded_Handle [{shapeType=Capsule, length=71.6693, radius=6.0212, center=<-0.4788 1.3519 24.9793>, rotation=<rot 4.7778 -90 0.0403>}]}
constraints []
